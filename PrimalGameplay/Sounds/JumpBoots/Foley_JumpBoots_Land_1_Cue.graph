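# Foley Sound Cue for a jump boots landing, layering step and spring sounds under player foley class, attenuation and concurrency settings.
# A quality switch gives the first two levels a richer mix and the last two a smaller wave pool with a single spring.
# Each mix feeds its random step branch in both directly and through a short random delay, doubling the step alongside a pitch-modulated spring.
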SoundCue /PrimalGameplay/Sounds/JumpBoots/Foley_JumpBoots_Land_1_Cue
root=Modulator_0
nodes [
 {id=Delay_0, children [Modulator_4], properties {delayMax=0.06, delayMin=0.03}}
 {id=Delay_1, children [Random_3], properties {delayMax=0.05, delayMin=0.02}}
 {id=Mixer_0, children [null,Delay_0,Modulator_4,Modulator_1], properties {inputVolume [0.75,1.1,0.75,1.4]}}
 {id=Mixer_2, children [null,Random_3,Delay_1,Modulator_2], properties {inputVolume [1,1,0.65,1]}}
 {id=Modulator_0, children [QualityLevel_0]}
 {id=Modulator_1, children [Random_2], properties {pitchMax=1.5, pitchMin=1.1}}
 {id=Modulator_2, children [WavePlayer_15], properties {pitchMax=1.2, pitchMin=0.9, volumeMin=0.75}}
 {id=Modulator_4, children [Random_4], properties {pitchMax=1.3, pitchMin=1.2}}
 {id=QualityLevel_0, children [Mixer_0,Mixer_0,Mixer_2,Mixer_2]}
 {id=Random_2, children [WavePlayer_11, WavePlayer_12, WavePlayer_13, WavePlayer_14], properties {weights [1,0.5,1,1]}}
 {id=Random_3, children [WavePlayer_33, WavePlayer_35, WavePlayer_36], properties {bShouldExcludeFromBranchCulling=true, weights [1,1,1]}}
 {
  id=Random_4
  children [WavePlayer_16, WavePlayer_17, WavePlayer_18, WavePlayer_19, WavePlayer_20, WavePlayer_21]
  properties {weights [1,1,1,1,1,1]}
 }
 {id=WavePlayer_11, properties {soundWaveAssetPtr=&/JumpBoots_Land_Spring_01}}
 {id=WavePlayer_12, properties {soundWaveAssetPtr=&/JumpBoots_Land_Spring_02}}
 {id=WavePlayer_13, properties {soundWaveAssetPtr=&/JumpBoots_Land_Spring_03}}
 {id=WavePlayer_14, properties {soundWaveAssetPtr=&/JumpBoots_Land_Spring_04}}
 {id=WavePlayer_15, properties {soundWaveAssetPtr=&JumpBoots_Land_Spring_01}}
 {id=WavePlayer_16, properties {soundWaveAssetPtr=&/JumpBoots_Step_01}}
 {id=WavePlayer_17, properties {soundWaveAssetPtr=&/JumpBoots_Step_04}}
 {id=WavePlayer_18, properties {soundWaveAssetPtr=&/JumpBoots_Step_05}}
 {id=WavePlayer_19, properties {soundWaveAssetPtr=&/JumpBoots_Step_06}}
 {id=WavePlayer_20, properties {soundWaveAssetPtr=&/JumpBoots_Step_07}}
 {id=WavePlayer_21, properties {soundWaveAssetPtr=&/JumpBoots_Step_08}}
 {id=WavePlayer_33, properties {soundWaveAssetPtr=&JumpBoots_Step_01}}
 {id=WavePlayer_35, properties {soundWaveAssetPtr=&/JumpBoots_Step_02}}
 {id=WavePlayer_36, properties {soundWaveAssetPtr=&/JumpBoots_Step_03}}
]
properties {
 attenuationSettings=&Foley_Attenuation_Basic_Att
 maxDistance=1536
 soundClassObject=&Fort_Player_Foley
 volumeMultiplier=0.55
 concurrencySet [
  /Game/Sounds/Fort_Foley_Sounds/FoleyLibraries/Settings/Foley_Basic_Local_Conc
  /Game/Sounds/Fort_Foley_Sounds/FoleyLibraries/Settings/Foley_Basic_Global_Conc
 ]
}
