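# Foley sound cue for a jump boots landing (Land 3 variant), using basic foley attenuation, concurrency limits and the player foley sound class.
# A root modulator wraps a quality switch between two mixes.
# The fuller mix layers random landing waves with random step waves, fed both through a short random delay and directly at reduced volume.
# The lighter mix uses one landing wave and a smaller step pool, wired twice without delay.
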SoundCue /PrimalGameplay/Sounds/JumpBoots/Foley_JumpBoots_Land_3_Cue
root=Modulator_0
nodes [
 {id=Delay_0, children [Random_0], properties {delayMax=0.05, delayMin=0.02}}
 {id=Mixer_0, children [Random_2,Delay_0,Random_0], properties {inputVolume [1,1,0.65]}}
 {id=Mixer_2, children [WavePlayer_14, Random_3, Random_3], properties {inputVolume [1,1,0.65]}}
 {id=Modulator_0, children [QualityLevel_0]}
 {id=QualityLevel_0, children [Mixer_0,Mixer_0,Mixer_2,Mixer_2]}
 {
  id=Random_0
  children [WavePlayer_3, WavePlayer_4, WavePlayer_5, WavePlayer_6, WavePlayer_7, WavePlayer_8, WavePlayer_9, WavePlayer_10]
  properties {weights [1,1,1,1,1,1,1,1]}
 }
 {id=Random_2, children [WavePlayer_11, WavePlayer_12, WavePlayer_13], properties {weights [1,1,1]}}
 {id=Random_3, children [WavePlayer_33, WavePlayer_35, WavePlayer_36], properties {bShouldExcludeFromBranchCulling=true, weights [1,1,1]}}
 {id=WavePlayer_10, properties {soundWaveAssetPtr=&/JumpBoots_Step_08}}
 {id=WavePlayer_11, properties {soundWaveAssetPtr=&/JumpBoots_Land_3_01}}
 {id=WavePlayer_12, properties {soundWaveAssetPtr=&/JumpBoots_Land_3_02}}
 {id=WavePlayer_13, properties {soundWaveAssetPtr=&/JumpBoots_Land_3_03}}
 {id=WavePlayer_14, properties {soundWaveAssetPtr=&JumpBoots_Land_3_01}}
 {id=WavePlayer_3, properties {soundWaveAssetPtr=&/JumpBoots_Step_01}}
 {id=WavePlayer_33, properties {soundWaveAssetPtr=&JumpBoots_Step_01}}
 {id=WavePlayer_35, properties {soundWaveAssetPtr=&/JumpBoots_Step_02}}
 {id=WavePlayer_36, properties {soundWaveAssetPtr=&/JumpBoots_Step_03}}
 {id=WavePlayer_4, properties {soundWaveAssetPtr=&JumpBoots_Step_02}}
 {id=WavePlayer_5, properties {soundWaveAssetPtr=&JumpBoots_Step_03}}
 {id=WavePlayer_6, properties {soundWaveAssetPtr=&/JumpBoots_Step_04}}
 {id=WavePlayer_7, properties {soundWaveAssetPtr=&/JumpBoots_Step_05}}
 {id=WavePlayer_8, properties {soundWaveAssetPtr=&/JumpBoots_Step_06}}
 {id=WavePlayer_9, properties {soundWaveAssetPtr=&/JumpBoots_Step_07}}
]
properties {
 attenuationSettings=&Foley_Attenuation_Basic_Att
 maxDistance=1536
 soundClassObject=&Fort_Player_Foley
 volumeMultiplier=0.8
 concurrencySet [
  /Game/Sounds/Fort_Foley_Sounds/FoleyLibraries/Settings/Foley_Basic_Local_Conc
  /Game/Sounds/Fort_Foley_Sounds/FoleyLibraries/Settings/Foley_Basic_Global_Conc
 ]
}
