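# Sound cue for the first jump sound of the jump boots gadget.
# A Random node picks one of four equally weighted wave variations.
# A Modulator at the root then varies pitch and volume per play to reduce repetition.
# It plays quietly through the gadgets and items sound class with a shared jump boot attenuation asset.
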SoundCue /PrimalGameplay/Sounds/JumpBoots/JumpBoots_Jump_1_Cue
root=Modulator_3
nodes [
 {id=Modulator_3, children [Random_4], properties {pitchMax=1.1, volumeMin=0.75}}
 {id=Random_4, children [WavePlayer_13, WavePlayer_15, WavePlayer_0, WavePlayer_1], properties {weights [1,1,1,1]}}
 {id=WavePlayer_0, properties {soundWaveAssetPtr=&/JumpBoots_Jump_1_02}}
 {id=WavePlayer_1, properties {soundWaveAssetPtr=&/JumpBoots_Jump_1_06}}
 {id=WavePlayer_13, properties {soundWaveAssetPtr=&/JumpBoots_Jump_1_01}}
 {id=WavePlayer_15, properties {soundWaveAssetPtr=&/JumpBoots_Jump_1_03}}
]
properties {
 attenuationSettings=/Game/Sounds/Attenuation/Fort_JumpBoot_Att
 maxDistance=3584
 soundClassObject=&Fort_Gadgets_Items
 volumeMultiplier=0.1
 attenuationOverrides {distanceAlgorithm=Logarithmic, falloffDistance=500, attenuationShapeExtents {x=0}}
}
