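# Sound cue for a Jump Boots jump sound, the second variant by name, routed to the gadget items sound class at reduced volume.
# A Random node picks between two equally weighted wave players, then a Modulator varies pitch slightly while holding volume constant.
# It references the Jump Boots attenuation asset and also carries attenuation override values with logarithmic falloff.
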SoundCue /PrimalGameplay/Sounds/JumpBoots/JumpBoots_Jump_2_Cue
root=Modulator_3
nodes [
 {id=Modulator_3, children [Random_0], properties {pitchMin=0.97, volumeMax=1, volumeMin=1}}
 {id=Random_0, children [WavePlayer_0,WavePlayer_1], properties {weights [1,1]}}
 {id=WavePlayer_0, properties {soundWaveAssetPtr=&/JumpBoots_Jump_2_01}}
 {id=WavePlayer_1, properties {soundWaveAssetPtr=&/JumpBoots_Jump_2_02}}
]
properties {
 attenuationSettings=/Game/Sounds/Attenuation/Fort_JumpBoot_Att
 maxDistance=3584
 soundClassObject=&Fort_Gadgets_Items
 volumeMultiplier=0.35
 attenuationOverrides {distanceAlgorithm=Logarithmic, falloffDistance=500, attenuationShapeExtents {x=0}}
}
