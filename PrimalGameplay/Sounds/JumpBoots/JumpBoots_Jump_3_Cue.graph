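# Sound cue for the third jump-boots jump sound, classed as a gadget item sound.
# A random node picks one of three equally weighted wave variations, then a modulator applies a slight upward pitch variation at constant volume.
# The cue plays at half volume through a shared jump-boot attenuation asset.
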SoundCue /PrimalGameplay/Sounds/JumpBoots/JumpBoots_Jump_3_Cue
root=Modulator_0
nodes [
 {id=Modulator_0, children [Random_0], properties {pitchMax=1.15, pitchMin=1, volumeMax=1, volumeMin=1}}
 {id=Random_0, children [WavePlayer_0,WavePlayer_1,WavePlayer_2], properties {weights [1,1,1]}}
 {id=WavePlayer_0, properties {soundWaveAssetPtr=&/JumpBoots_Jump_3_01}}
 {id=WavePlayer_1, properties {soundWaveAssetPtr=&/JumpBoots_Jump_3_02}}
 {id=WavePlayer_2, properties {soundWaveAssetPtr=&/JumpBoots_Jump_3_03}}
]
properties {
 attenuationSettings=/Game/Sounds/Attenuation/Fort_JumpBoot_Att
 maxDistance=3584
 soundClassObject=&Fort_Gadgets_Items
 volumeMultiplier=0.5
 attenuationOverrides {distanceAlgorithm=Logarithmic, falloffDistance=500, attenuationShapeExtents {x=0}}
}
