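# Sound cue for picking up Jump Boots, judging by its name and loot-collect sound class.
# A single JumpBoots_Pickup_01 wave player feeds a modulator at the root, which is left at its default settings.
# It plays at reduced volume with shared weapon-pickup attenuation and the Fort_Loot_Collect sound class.
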SoundCue /PrimalGameplay/Sounds/JumpBoots/JumpBoots_Pickup_Cue
root=Modulator_0
nodes [
 {id=Modulator_0, children [WavePlayer_1]}
 {id=WavePlayer_1, properties {soundWaveAssetPtr=&/JumpBoots_Pickup_01}}
]
properties {
 attenuationSettings=&Fort_Weapon_Pickup_Att
 maxDistance=3500
 soundClassObject=&Fort_Loot_Collect
 volumeMultiplier=0.7
}
